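# Lager_Amb_Battle_Cue is a looping ambient sound cue that plays distant battle noise for Lager encounters.
# A single looping wave player feeds an Enveloper that fades volume from silent to full over the first three seconds.
# The cue uses a dedicated battle attenuation asset and routes through the distant ranged weapon fire sound class.
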SoundCue /Lager/Sounds/SFX/Encounters/Lager_Amb_Battle_Cue
root=Enveloper_0
nodes [
 {
  id=Enveloper_0
  children [WavePlayer_6]
  properties {pitchCurve {editorCurveData {keys [{time=0}]}}, volumeCurve {editorCurveData {keys [<at 0 0>,<at 3 1>]}}}
 }
 {id=WavePlayer_6, properties {bLooping=true, soundWaveAssetPtr=&/Lager_Amb_Dist_Battle_Loop_01}}
]
properties {
 attenuationSettings=&/Lager_Battle_Att
 maxDistance=20000
 soundClassObject=&Fort_WeaponRanged_Fire_Distant
}
