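# Sound cue for an ambient distant battle-start sound in Lager encounters.
# A Random node picks one of four equally weighted wave variations of the distant battle start.
# It plays at reduced volume through a battle-start attenuation asset and uses the distant ranged-weapon fire sound class.
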SoundCue /Lager/Sounds/SFX/Encounters/Lager_Amb_Battle_Start_Cue
root=Random_0
nodes [
 {id=Random_0, children [WavePlayer_0,WavePlayer_2,WavePlayer_3,WavePlayer_4], properties {weights [1,1,1,1]}}
 {id=WavePlayer_0, properties {soundWaveAssetPtr=&/Lager_Amb_Dist_Battle_Starts_01}}
 {id=WavePlayer_2, properties {soundWaveAssetPtr=&/Lager_Amb_Dist_Battle_Starts_02}}
 {id=WavePlayer_3, properties {soundWaveAssetPtr=&/Lager_Amb_Dist_Battle_Starts_03}}
 {id=WavePlayer_4, properties {soundWaveAssetPtr=&/Lager_Amb_Dist_Battle_Starts_04}}
]
properties {
 attenuationSettings=&/Lager_Battle_Start_Att
 maxDistance=20000
 soundClassObject=&Fort_WeaponRanged_Fire_Distant
 volumeMultiplier=0.6
}
