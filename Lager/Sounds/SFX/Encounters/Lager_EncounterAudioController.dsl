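; Lager encounter audio controller that plays ambient combat audio for fights between attacker and defender pawns.
; Attacker and defender updates maintain a combined pawn list with parallel per-pawn data, then try starting a timed update.
; Updates move a looping sound to the pawns' centroid, with a one-time starting stinger; stopping fades the loop out.
; Each pawn's weapon audio components are cached once and given attenuation and concurrency overrides plus a low-pass filter.
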(fn ApplyGunfireMix
  (params (Target :type Fort%20Pawn%20Object%20Reference) (DataIndex :type Integer))
  (bind _returnvalue (Weapon|GetCurrentWeapon Target))
  (bind _returnvalue_2 (Actor|GetComponentsByClass _returnvalue "/Script/Engine.AudioComponent"))
  (bind _output_audiocomponents_3_c97e161049d00a2525afc8b1a53f07be (Utilities|Struct|BreakLagerEncounterAudioActorData (Utilities|Array|Get(aref) (Variables|Default|GetEncounterData) (Variables|GetDataIndex))))
  (Utilities|IsValid Target
    (:Is%20Valid
      (Utilities|IsValid _returnvalue
        (:Is%20Valid
          (if (Utilities|Array|IsEmpty _output_audiocomponents_3_c97e161049d00a2525afc8b1a53f07be)
            (if (Utilities|Array|IsNotEmpty _returnvalue_2)
              (bind _structout (Utilities|Struct|SetmembersinLagerEncounterAudioActorData (Utilities|Array|Get(aref) (Variables|Default|GetEncounterData) (Variables|GetDataIndex)) _returnvalue_2))
              (bind _structout_audiocomponents_3_c97e161049d00a2525afc8b1a53f07be (Utilities|Struct|BreakLagerEncounterAudioActorData _structout))
              (for _array_element _structout_audiocomponents_3_c97e161049d00a2525afc8b1a53f07be
                (Utilities|IsValid _array_element
                  (:Is%20Valid
                    (Utilities|FlowControl|Sequence
                      (:then_0
                        (Variables|/Script/Engine.AudioComponent|SetAttenuationSettings (Variables|Default|GetGunfireAttenuationOverride) _array_element))
                      (:then_1
                        (Variables|/Script/Engine.AudioComponent|SetConcurrencySet (Variables|Default|GetGunfireConcurrencyOverride) _array_element))
                      (:then_2
                        (Audio|Components|Audio|SetLowPassFilterEnabled _array_element true :!class /Script/Engine.AudioComponent))
                      (:then_3
                        (Audio|Components|Audio|SetLowPassFilterFrequency _array_element 5e+03 :!class /Script/Engine.AudioComponent)))))))))))))

(fn GetInvalidIndices
  (params (Array :type Array%20of%20Object%20References))
  (bind _indexes_to_remove (LocalVariables|GetIndexes%20to%20remove))
  (for (_array_element _array_index) Array
    (Utilities|IsValid _array_element
      (:Is%20Not%20Valid
        (Utilities|Array|Add (LocalVariables|GetIndexes%20to%20remove) _array_index))))
  (Utilities|Array|Reverse _indexes_to_remove)
  (return _indexes_to_remove))

(fn OnUpdate
  (if (not (or (Utilities|Array|IsEmpty (Variables|Default|GetAttackers)) (Utilities|Array|IsEmpty (Variables|Default|GetDefenders))))
    (bind _returnvalue (Transformation|GetActorArrayAverageLocation (Variables|Default|GetEncounterPawns)))
    (bind _output_get (Variables|Default|SetCentroid _returnvalue))
    (if (not (Math|Vector|VectorIsNearlyZero _output_get))
      (Variables|Default|GetAudioComponent
        (:then
          _)
        (:else
          (bind _returnvalue_2 (Audio|SpawnSoundatLocation (Variables|Default|GetLoopSoundAsset) (Variables|Default|GetCentroid)))
          (Variables|Default|SetAudioComponent _returnvalue_2)
          (if (not (Variables|Default|GetbHasPlayedStartingStinger))
            (Audio|PlaySoundatLocation (Variables|Default|GetStartSoundAsset) (Variables|Default|GetCentroid))
            (Variables|Default|SetbHasPlayedStartingStinger true))))
      (Transformation|SetWorldLocation (Variables|Default|GetAudioComponent) (Variables|Default|GetCentroid) false false)
      (for (_array_element _array_index) (Variables|Default|GetEncounterPawns)
        (CallFunction|ApplyGunfireMix _array_element _array_index))))
  (if (Variables|Default|GetbUpdating)
    (Default|StopUpdating)))

(fn StopUpdating
  (Variables|Default|SetbUpdating false)
  (Utilities|Time|ClearandInvalidateTimerbyHandle (Variables|Default|GetUpdateTimerHandle))
  (bind _audiocomponent (Variables|Default|GetAudioComponent)
    (:then
      (Audio|Components|Audio|FadeOut _audiocomponent (Variables|Default|GetFadeOutTime) 0.0 "Linear" :!class /Script/Engine.AudioComponent))
    (:else)))

(fn TryStartUpdating
  (if (not (Variables|Default|GetbUpdating))
    (if (and (Utilities|Array|IsNotEmpty (Variables|Default|GetAttackers)) (Utilities|Array|IsNotEmpty (Variables|Default|GetDefenders)))
      (bind _returnvalue (Utilities|Time|SetTimerbyEvent (EventDispatchers|CreateEvent :!event OnUpdate) (Variables|Default|GetUpdateFrequency) true))
      (Variables|Default|SetUpdateTimerHandle _returnvalue)
      (Variables|Default|SetbUpdating true)
      (Default|OnUpdate)
      (else
        (Default|StopUpdating)))))

(fn UpdateAttackers
  (params (Attackers :type Array%20of%20Fort%20Pawn%20Object%20References))
  (bind _output_get (Variables|Default|SetAttackers Attackers))
  (for _array_element _output_get
    (Utilities|Array|AddUnique (Variables|Default|GetEncounterPawns) _array_element))
  (Utilities|Array|Resize (Variables|Default|GetEncounterData) (Utilities|Array|Length (Variables|Default|GetEncounterPawns)))
  (bind _indices (CallFunction|GetInvalidIndices (Variables|Default|GetEncounterPawns)))
  (for _array_element_2 _indices
    (Utilities|Array|RemoveIndex (Variables|Default|GetEncounterPawns) _array_element_2)
    (Utilities|Array|RemoveIndex (Variables|Default|GetEncounterData) _array_element_2))
  (Default|TryStartUpdating))

(fn UpdateDefenders
  (params (Defenders :type Array%20of%20Fort%20Pawn%20Object%20References))
  (bind _output_get (Variables|Default|SetDefenders Defenders))
  (for _array_element _output_get
    (Utilities|Array|AddUnique (Variables|Default|GetEncounterPawns) _array_element))
  (Utilities|Array|Resize (Variables|Default|GetEncounterData) (Utilities|Array|Length (Variables|Default|GetEncounterPawns)))
  (bind _indices (CallFunction|GetInvalidIndices (Variables|Default|GetEncounterPawns)))
  (for _array_element_2 _indices
    (Utilities|Array|RemoveIndex (Variables|Default|GetEncounterPawns) _array_element_2)
    (Utilities|Array|RemoveIndex (Variables|Default|GetEncounterData) _array_element_2))
  (Default|TryStartUpdating))
