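# Sound cue for the Lager pod charge-up effect, apparently used in encounters.
# A single wave player plays one charge-up sound wave with boosted volume.
# It routes to the enemy attack sound class and uses the pod spawn attenuation settings.
# A concurrency override caps how many instances play at once.
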SoundCue /Lager/Sounds/SFX/Encounters/Lager_Pod_Charge_Up_Cue
root=WavePlayer_1
nodes [{id=WavePlayer_1, properties {soundWaveAssetPtr=&/Lager_Pod_Charge_Up_02}}]
properties {
 attenuationSettings=/Lager/Sounds/Settings/Lager_Pod_Spawn_Att
 bOverrideConcurrency=true
 maxDistance=7500
 soundClassObject=/Game/Sounds/FortSoundClassesAndModes/Fort_Enemy_Attack
 volumeMultiplier=1.5
 concurrencyOverrides {maxCount=2}
}
