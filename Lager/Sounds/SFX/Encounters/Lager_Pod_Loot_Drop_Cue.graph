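# Sound cue for the loot drop from a Lager encounter pod, named as a pod loot-drop effect.
# It plays a single wave, Lager_Pod_Loot_Drop_01, with no randomization or mixing.
# It reuses the Fort weapon-drop attenuation and the Fort loot-collect sound class, at slightly reduced volume.
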SoundCue /Lager/Sounds/SFX/Encounters/Lager_Pod_Loot_Drop_Cue
root=WavePlayer_1
nodes [{id=WavePlayer_1, properties {soundWaveAssetPtr=&/Lager_Pod_Loot_Drop_01}}]
properties {
 attenuationSettings=&Fort_Weapon_Drop_Att
 maxDistance=3000
 soundClassObject=&Fort_Loot_Collect
 volumeMultiplier=0.796
}
