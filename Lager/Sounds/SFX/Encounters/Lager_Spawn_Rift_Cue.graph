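# Sound cue for the Lager encounter spawn rift effect, classed as an enemy attack sound.
# A Random node picks one of eight equally weighted rift wave variations, and a Modulator at the root varies its playback.
# It plays boosted in volume with AI spawn attenuation and its own concurrency override, which allows only a couple of simultaneous instances, stops the farthest first and limits rapid retriggering.
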SoundCue /Lager/Sounds/SFX/Encounters/Lager_Spawn_Rift_Cue
root=Modulator_0
nodes [
 {id=Modulator_0, children [Random_0]}
 {
  id=Random_0
  children [WavePlayer_0,WavePlayer_1,WavePlayer_2,WavePlayer_3,WavePlayer_4,WavePlayer_5,WavePlayer_6,WavePlayer_7]
  properties {weights [1,1,1,1,1,1,1,1]}
 }
 {id=WavePlayer_0, properties {soundWaveAssetPtr=&/Lager_Spawn_Rift_08}}
 {id=WavePlayer_1, properties {soundWaveAssetPtr=&/Lager_Spawn_Rift_01}}
 {id=WavePlayer_2, properties {soundWaveAssetPtr=&/Lager_Spawn_Rift_02}}
 {id=WavePlayer_3, properties {soundWaveAssetPtr=&/Lager_Spawn_Rift_03}}
 {id=WavePlayer_4, properties {soundWaveAssetPtr=&/Lager_Spawn_Rift_04}}
 {id=WavePlayer_5, properties {soundWaveAssetPtr=&/Lager_Spawn_Rift_05}}
 {id=WavePlayer_6, properties {soundWaveAssetPtr=&/Lager_Spawn_Rift_06}}
 {id=WavePlayer_7, properties {soundWaveAssetPtr=&/Lager_Spawn_Rift_07}}
]
properties {
 attenuationSettings=/Lager/Sounds/Settings/Lager_AI_Spawn_Att
 bOverrideConcurrency=true
 maxDistance=6200
 soundClassObject=/Game/Sounds/FortSoundClassesAndModes/Fort_Enemy_Attack
 volumeMultiplier=1.5
 concurrencyOverrides {maxCount=2, resolutionRule=StopFarthestThenPreventNew, retriggerTime=0.1}
}
