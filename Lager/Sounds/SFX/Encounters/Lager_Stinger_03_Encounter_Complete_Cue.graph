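# Sound cue that plays the encounter-complete stinger for Lager encounters.
# A single wave player feeds the cue with no randomization or modulation, slightly boosted in volume.
# It routes to the critical in-game UI sound class with a very large max distance.
# A concurrency override allows only one instance at a time and blocks new ones while it plays.
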SoundCue /Lager/Sounds/SFX/Encounters/Lager_Stinger_03_Encounter_Complete_Cue
root=WavePlayer_0
nodes [{id=WavePlayer_0, properties {soundWaveAssetPtr=&/Lager_Stinger_03_Encounter_Complete_01}}]
properties {
 bOverrideConcurrency=true
 maxDistance=2097152
 soundClassObject=&Fort_UI_InGame_Critical
 volumeMultiplier=1.25
 concurrencyOverrides {maxCount=1, resolutionRule=PreventNew}
}
